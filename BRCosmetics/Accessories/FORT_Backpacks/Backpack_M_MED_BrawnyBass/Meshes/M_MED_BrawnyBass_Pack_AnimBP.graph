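# Animation Blueprint for the BrawnyBass backpack cosmetic, bound to that pack's own skeleton.
# The AnimGraph feeds an Input Pose straight to the Output Pose, adding no animation of its own.
# It declares a parameterless SpawnFX_Event dispatcher, judging by the matching variable and empty signature graph, but nothing here calls or binds it.
# The EventGraph is empty.
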AnimBlueprint /BRCosmetics/Accessories/FORT_Backpacks/Backpack_M_MED_BrawnyBass/Meshes/M_MED_BrawnyBass_Pack_AnimBP
target_skeleton=&/M_MED_BrawnyBass_Pack_Skeleton
graphs [
 AnimGraph {nodes [Output Pose(Result=Input Pose()) {layerGroup=None}]}
 EventGraph {links [], nodes []}
 SpawnFX_Event {nodes [SpawnFX_Event() {is=title K2Node_FunctionEntry}]}
]
variables [SpawnFX_Event {default=()}]
